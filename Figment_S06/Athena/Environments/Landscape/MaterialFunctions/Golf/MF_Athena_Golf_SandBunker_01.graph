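# Material function that supplies the sand bunker surface for the Athena golf landscape.
# Tiled, rotatable UV1 coordinates sample sand color, normal and a packed mask, whose red channel blends min and max roughness and green drives opacity, forming the UVSand attributes.
# UV0 separately samples a second normal map and the packed mask's blue channel, output as BunkerAssetNormal and BunkerAssetMask for per-asset detail.
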MaterialFunction /Figment_S06/Athena/Environments/Landscape/MaterialFunctions/Golf/MF_Athena_Golf_SandBunker_01
graph {nodes [
  n1 = MaterialFunctionCall(
   UVs=Multiply(A=TextureCoordinate() {coordinateIndex=1}, B=ScalarParameter() {defaultValue=1, parameterName=BunkerSand_Tiling})
   "Rotation Angle (0-1)"=ScalarParameter() {parameterName=UVRotation90}
  ) {materialFunction=&CustomRotator}
  n2 = TextureSample(UVs=@n1.Rotated Values) {samplerSource=SSM_Wrap_WorldGroupSettings, texture=/Figment_S06/Athena/Environments/Landscape/Textures/Golf/T_Athena_SandBunker_01_D}
  n3 = ScalarParameter() {parameterName=RoughnessMin}
  n4 = ScalarParameter() {defaultValue=1, parameterName=RoughnessMax}
  n5 = TextureSample(UVs=@n1.Rotated Values) {
   samplerSource=SSM_Wrap_WorldGroupSettings
   samplerType=SAMPLERTYPE_LinearColor
   texture=/Figment_S06/Athena/Environments/Landscape/Textures/Golf/T_Athena_SandBunker_01_M
  }
  n6 = TextureSample(UVs=@n1.Rotated Values) {
   samplerSource=SSM_Wrap_WorldGroupSettings
   samplerType=SAMPLERTYPE_Normal
   texture=/Game/Athena/Environments/Landscape/Textures/Golf/T_Athena_SandBunker_01_N
  }
  FunctionOutput(None=MakeMaterialAttributes!(
    BaseColor=@n2.RGB
    Metallic
    Specular
    Roughness=LinearInterpolate(A=@n3, B=@n4, Alpha=@n5.R)
    Anisotropy
    EmissiveColor
    Opacity=@n5.G
    OpacityMask
    Normal=@n6.RGB
    Tangent
    WorldPositionOffset
    SubsurfaceColor
    ClearCoat
    ClearCoatRoughness
    AmbientOcclusion
    Refraction
    CustomizedUVs_0
    CustomizedUVs_1
    CustomizedUVs_2
    CustomizedUVs_3
    CustomizedUVs_4
    CustomizedUVs_5
    CustomizedUVs_6
    CustomizedUVs_7
    PixelDepthOffset
    ShadingModel
    Displacement
   )) {outputName=UVSand}
  n7 = TextureCoordinate()
  FunctionOutput(None=TextureSample(UVs=@n7) {samplerType=SAMPLERTYPE_Normal, texture=/Figment_S06/Athena/Environments/Landscape/Textures/Golf/T_Athena_SandBunker_01_N_2}.RGB) {outputName=BunkerAssetNormal}
  n8 = TextureSample(UVs=@n7) {
   samplerSource=SSM_Wrap_WorldGroupSettings
   samplerType=SAMPLERTYPE_LinearColor
   texture=&T_Athena_SandBunker_01_M
  }
  FunctionOutput(None=@n8.B) {outputName=BunkerAssetMask}
 ]}
